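# Masked lit material for the Creative music note coin pickup, with switchable diffuse, specular mask and normal inputs.
# Emissive sums a vertex-red ring glow, Fresnel rim and animated camera-aligned sweep boosted by vertex green, plus an optional emissive map.
# Normals oscillate through FixRotateAboutAxisNormals, noise drives the opacity mask, and attributes pass through MF_CreativeMoveTool.
# Bob, spin and black-hole offset nodes exist but are unconnected to World Position Offset.
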Material /Game/Creative/Devices/Creative_Coin/MusicNote_Pickup/Materials/M_CP_MusicNote_Pickup_01
graph {
 properties {
  bTangentSpaceNormal=false
  bUseMaterialAttributes=true
  bUsedWithNiagaraMeshParticles=true
  blendMode=BLEND_Masked
  pixelDepthOffsetMode=PDOM_Legacy
  translucencyLightingMode=TLM_SurfacePerPixelLighting
  usedShadingModels=Default Lit
 }
 nodes [
  n1 = Multiply(A=Time(), B=ScalarParameter() {defaultValue=0.4, parameterName=WPOTimeScale})
  n2 = LinearInterpolate(
   A=VectorParameter() {parameterName=BobHeightMax}.RGB
   B=VectorParameter() {parameterName=BobHeightMin, defaultValue=<rgba 0 0 -2 0>}.RGB
   Alpha=Sine(None=@n1)
  )
  n3 = Constant3Vector() {constant=<rgba 0 0 1 0>}
  n4 = RotateAboutAxis(
   NormalizedRotationAxis=@n3
   RotationAngle=Multiply(A=Time(), B=ScalarParameter() {defaultValue=0.1, parameterName=RotationSpeed})
   PivotPoint=ObjectPositionWS()
   Position=WorldPosition().XYZ
  )
  n5 = VectorParameter() {parameterName=WorldPostionOfBlackHole, defaultValue=<rgba -30000 25536 1952 0>}
  Add(
   A=StaticSwitchParameter(True=Constant(), False=Add(A=@n2, B=@n4)) {parameterName=Use WPO}
   B=MaterialFunctionCall!(WorldPostionOfBlackHole=AppendVector(A=@n5.RGB, B=@n5.RGB), World Position) {materialFunction=/Game/Packages/Fortnite_UniqueMaterials_1/BlackHoleFunction}.Result
  )
  Constant() {r=0.001}
  MaterialFunctionCall() {materialFunction=&ComponentPivotLocation}
  n6 = VectorParameter() {parameterName=BaseColor, defaultValue=<rgba 1 1 1 0>}
  n7 = StaticSwitchParameter(
   True=Multiply(
    A=TextureSampleParameter2D() {parameterName=Diffuse, texture=/Game/Creative/Devices/Creative_Coin/MusicNote_Pickup/Materials/Textures/T_CP_MusicNote_Pickup_01_D}.RGB
    B=@n6.RGB
   )
   False=@n6.RGB
  ) {defaultValue=true, parameterName=Use Diffuse}
  n8 = TextureSampleParameter2D() {
   parameterName=Specular Mask
   samplerType=SAMPLERTYPE_Masks
   texture=/Game/Creative/Devices/Creative_Coin/MusicNote_Pickup/Materials/Textures/T_CP_MusicNote_Pickup_01_S
  }
  n9 = Multiply(
   A=TextureSampleParameter2D() {group=Emissive, parameterName=EmissiveMap, texture=&T_BlacksRGB}.RGB
   B=VectorParameter() {group=Emissive, parameterName=EmissiveMultiplier, defaultValue=<rgba 1 1 1 0>}.RGB
  )
  n10 = Multiply(A=VertexColor().R, B=VectorParameter() {parameterName=RingGlow, defaultValue=<rgba 5 1.582 0.4427 0>}.RGB)
  n11 = Power(Base=@n7, Exp=ScalarParameter() {defaultValue=1, parameterName=DiffuseToEmissiveExponent})
  n12 = ScalarParameter() {parameterName=Rotation Angle Min}
  n13 = ScalarParameter() {defaultValue=0.3, parameterName=Rotation Angle Max}
  n14 = Cosine(None=@n1)
  n15 = VectorParameter() {parameterName=WorldPostionOfBlackHole, defaultValue=<rgba -30000 25536 1952 0>}
  n16 = TextureSampleParameter2D() {parameterName=Normal, samplerType=SAMPLERTYPE_Normal, texture=/Game/Athena/Playlists/Score/Coins/Textures/T_Coin_N}
  n17 = Constant3Vector() {constant=<rgba 0 0 1 0>}
  n18 = MaterialFunctionCall(
   Rotation Axis=@n3
   Rotation Angle=Multiply(A=LinearInterpolate(A=@n12, B=@n13, Alpha=@n14), B=OneMinus(None=@n15.A))
   World Space Vertex Normals=Transform(None=StaticSwitchParameter(True=@n16.RGB, False=@n17) {parameterName=Use Normal map?})
  ) {materialFunction=&FixRotateAboutAxisNormals}
  n19 = Clamp(None=DotProduct(A=CameraVectorWS(), B=@n18.New World Space Vertex Normal))
  n20 = Multiply(
   A=Power(Base=OneMinus(None=@n19)) {constExponent=4}
   B=VectorParameter() {parameterName=GlowColour, defaultValue=<rgba 9.6244 10 9.6475 0>}.RGB
  )
  n21 = Constant()
  n22 = Multiply(A=Subtract(A=WorldPosition().XYZ, B=ObjectPositionWS()), B=ScalarParameter() {defaultValue=0.001, parameterName=WorldTileScale})
  n23 = Normalize(VectorInput=Transform(None=VectorParameter() {parameterName=SweepDirectionCamSpace, defaultValue=<rgba -0.5 0.45 -1 0>}.RGB) {transformSourceType=TRANSFORMSOURCE_Camera})
  n24 = ComponentMask(None=Multiply(A=DotProduct(A=@n22, B=@n23), B=@n23)) {b=true, r=true}
  n25 = Constant() {r=0.28}
  n26 = Frac(None=ComponentMask(None=Panner(Coordinate=@n24, Speed=@n25) {speedX=0.2, speedY=0.2}) {g=true})
  n27 = Add(A=@n26) {constB=0.1}
  n28 = OneMinus(None=Multiply(A=Subtract(A=@n26) {constB=0.9}) {constB=10})
  n29 = MaterialFunctionCall() {materialFunction=&MF_Fresnel_FX}
  n30 = Multiply(A=Multiply(A=Floor(None=@n27), B=Power(Base=@n28) {constExponent=3}), B=Saturate(None=Add(A=@n29.Fresnel Inverted) {constB=0.1}))
  n31 = VectorParameter() {parameterName=SweepColour, defaultValue=<rgba 1.8958 2 1.9078 0>}
  n32 = LinearInterpolate(A=@n31.RGB, B=Multiply(A=@n31.RGB, B=ScalarParameter() {defaultValue=15, parameterName=FBoostScale}), Alpha=VertexColor().G)
  n33 = Add(
   A=Add(A=Multiply(A=@n10, B=@n11), B=StaticSwitchParameter(True=@n20, False=@n21) {defaultValue=true, parameterName=UseOuterGlow})
   B=StaticSwitchParameter(True=Multiply(A=@n30, B=@n32), False=Constant()) {defaultValue=true, parameterName=UseSweep}
  )
  n34 = TextureSample() {texture=&T_NoisePack_01}
  n35 = ScalarParameter() {defaultValue=1, parameterName=Opacity Noise Influence}
  n36 = ScalarParameter() {defaultValue=1, parameterName=Opacity}
  n37 = Subtract(A=Add(A=LinearInterpolate(B=@n34.R, Alpha=@n35) {constA=1}), B=Multiply(A=OneMinus(None=@n36)) {constB=2})
  n38 = SetMaterialAttributes!(
   MaterialAttributes
   Base Color=@n7
   Metallic=StaticSwitchParameter(True=@n8.G, False=Constant() {r=1}) {parameterName=Use Specular Masks}
   Specular=StaticSwitchParameter(True=@n8.R, False=Constant() {r=0.9}) {parameterName=Use Specular Masks}
   Roughness=StaticSwitchParameter(True=@n8.B, False=Constant() {r=0.1}) {parameterName=Use Specular Masks}
   Emissive Color=StaticSwitchParameter(True=Add(A=@n9, B=@n33), False=@n33) {group=Emissive, parameterName=UseEmissiveMap}
   Opacity Mask=@n37
   Normal=@n18.New World Space Vertex Normal
   World Position Offset
   Opacity=@n37
  ) {attributeSetTypes [
    #1
    #2
    #3
    #4
    #5
    #6
    #7
    #8
    #9
   ]}
  MaterialFunctionCall(MA In=@n38) {materialFunction=&MF_CreativeMoveTool}
 ]
}
